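# Niagara effect for the Butter Jack player body cosmetic that throws small bursts of green-tinted mesh particles off the character's skeletal mesh.
# Two near-identical emitters, one GPU and one CPU, sample spawn positions from a user-supplied skeletal mesh and launch particles outward with random velocity.
# Particles fall under drag and downward acceleration, partly inherit source movement, bounce weakly on collision, and shrink away late in life.
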NiagaraSystem /BRCosmetics/Characters/Player/Male/Medium/Bodies/M_MED_Butter_Jack/FX/NS_Butter_Jack
dependencies {usedDataInterfaces [&NiagaraDataInterfaceCurve], usedDynamicInputs [], usedModules [&SystemState]}
emitters {
 Empty {
  simTarget=GPUComputeSim
  emitterUpdateScript [
   &EmitterState(Inactive Response=::NewEnumerator1, Life Cycle Mode=::NewEnumerator1, Loop Behavior=::NewEnumerator1, Loop Duration=1, UseLoopDelay=false)
   &SpawnBurst_Instantaneous(
    Spawn Count=&RandomRangeInt(Minimum=3, Maximum=6)
    Use Spawn Probability=false
    UseLoopCountLimit=false
   )
   &SpawnRate(SpawnRate=30, Use Spawn Probability=false) {enabled=false}
  ]
  particleSpawnScript [
   &InitializeParticle(
    Color=<rgba 0.1668 0.4948 0.0495 1>
    Lifetime Max=1.2
    Lifetime Min=0.6
    Lifetime Mode=::NewEnumerator1
    Mesh Scale Mode=::NewEnumerator1
    Mesh Uniform Scale Max=1.3
    Mesh Uniform Scale Min=0.6
    UsePositionOffset=false
    Write Material Random=true
   )
   &InitialMeshOrientation(Use Rotation Vector=false)
   &CylinderLocation(
    Cylinder Height=0
    Cylinder Height Midpoint=1
    Cylinder Origin:NiagaraPosition=$StackContext.SkeletalMeshLocation.SampledPosition
    Cylinder Radius=10
    MaskSpawnGroup=false
    Override Seed=false
    UseOwnerScale=true
   ) {enabled=false}
   &SkeletalMeshLocation(
    Bone Sampling Mode=::NewEnumerator4
    Interpolate Mesh Sampling=false
    MaskSpawnGroup=false
    Sampled Position Offset=<5 0 0>
    Skeletal Mesh=$User.NewNiagaraDataInterfaceSkeletalMesh
   )
   &AddVelocityFromPoint(Module.Use Falloff Distance=false, Velocity Strength=&RandomRangeFloat(Minimum=80, Maximum=140))
   &AddVelocity(Velocity=&RandomRangeVector(Minimum=<-100 -100 0>, Maximum=<100 100 1>))
  ]
  particleUpdateScript [
   &Drag(Use Linear Drag=true, Use Rotational Drag=false)
   &AccelerationForce(Acceleration=&RandomRangeVector(Minimum=<0 0 -700>, Maximum=<0 0 -950>))
   &InheritSourceMovement(
    Applied Position Falloff Curve=DataInterface() {propertyValues {Curve {keys []}}}
    Applied Position Falloff Distance=80
    Applied Position Scale=<0.3 0.3 0.3>
    ApplyForce=true
    ApplyForceFalloff=false
    ApplyPosition=true
    ApplyPositionFalloff=true
    ApplyPositionFalloffCurve=false
    Module.LimitForceSpeed=false
    Source Velocity:Vector3f=$StackContext.Velocity
   )
   &Collision(
    Age Colliding Particles=false
    GPU Collision Type=::NewEnumerator2
    Module.Control Rest State Via Surface Normal=false
    Module.EnableMaxCollisionCount=false
    Module.Randomize Collision Normal Vector=false
    Particle Radius=0.3
    Particle Radius Scale=0.2
    Radius Calculation Type=::NewEnumerator3
    Restitution=0.2
   )
   &SolveForcesAndVelocity(Clamp Velocity=false, Limit Acceleration=false)
   &ParticleState()
   &UpdateMeshOrientation(Rotation Vector=<1 1 1>)
   &ScaleMeshSize(Scale Factor=&Multiply_VectorByFloat(
     Vector=<1 1 1>
     Float=&FloatFromCurve(FloatCurve=DataInterface() {propertyValues {Curve {keys [<at 0.78 1>,<at 1 0>]}}})
    ))
  ]
  rendererClasses [&NiagaraMeshRendererProperties]
  renderers [{rendererClass=&NiagaraMeshRendererProperties, rendererIndex=0}]
 }
 Empty001 {
  emitterUpdateScript [
   &EmitterState(Inactive Response=::NewEnumerator1, Life Cycle Mode=::NewEnumerator1, Loop Behavior=::NewEnumerator1, Loop Duration=1, UseLoopDelay=false)
   &SpawnBurst_Instantaneous(
    Spawn Count=&RandomRangeInt(Minimum=3, Maximum=4)
    Use Spawn Probability=false
    UseLoopCountLimit=false
   )
   &SpawnRate(SpawnRate=30, Use Spawn Probability=false) {enabled=false}
  ]
  particleSpawnScript [
   &InitializeParticle(
    Color=<rgba 0.1668 0.4948 0.0495 1>
    Lifetime Max=1.2
    Lifetime Min=0.6
    Lifetime Mode=::NewEnumerator1
    Mesh Scale Mode=::NewEnumerator1
    Mesh Uniform Scale Max=1.3
    Mesh Uniform Scale Min=0.6
    UsePositionOffset=false
    Write Material Random=true
   )
   &InitialMeshOrientation(Use Rotation Vector=false)
   &CylinderLocation(
    Cylinder Height=0
    Cylinder Height Midpoint=1
    Cylinder Origin:NiagaraPosition=$StackContext.SkeletalMeshLocation.SampledPosition
    Cylinder Radius=10
    MaskSpawnGroup=false
    Override Seed=false
    UseOwnerScale=true
   ) {enabled=false}
   &SkeletalMeshLocation(
    Bone Sampling Mode=::NewEnumerator4
    Interpolate Mesh Sampling=false
    MaskSpawnGroup=false
    Sampled Position Offset=<5 0 0>
    Skeletal Mesh=$User.NewNiagaraDataInterfaceSkeletalMesh
   )
   &AddVelocityFromPoint(Module.Use Falloff Distance=false, Velocity Strength=&RandomRangeFloat(Minimum=80, Maximum=140))
   &AddVelocity(Velocity=&RandomRangeVector(Minimum=<-100 -100 0>, Maximum=<100 100 1>))
  ]
  particleUpdateScript [
   &Drag(Use Linear Drag=true, Use Rotational Drag=false)
   &AccelerationForce(Acceleration=&RandomRangeVector(Minimum=<0 0 -700>, Maximum=<0 0 -950>))
   &InheritSourceMovement(
    Applied Position Falloff Curve=DataInterface() {propertyValues {Curve {keys []}}}
    Applied Position Falloff Distance=80
    Applied Position Scale=<0.3 0.3 0.3>
    ApplyForce=true
    ApplyForceFalloff=false
    ApplyPosition=true
    ApplyPositionFalloff=true
    ApplyPositionFalloffCurve=false
    Module.LimitForceSpeed=false
    Source Velocity:Vector3f=$StackContext.Velocity
   )
   &Collision(
    Age Colliding Particles=false
    Max CPU Trace Length=18
    Module.Control Rest State Via Surface Normal=false
    Module.EnableMaxCollisionCount=false
    Module.Randomize Collision Normal Vector=false
    Particle Radius=0.3
    Particle Radius Scale=0.2
    Radius Calculation Type=::NewEnumerator3
    Restitution=0.2
   )
   &SolveForcesAndVelocity(Clamp Velocity=false, Limit Acceleration=false)
   &ParticleState()
   &UpdateMeshOrientation(Rotation Vector=<1 1 1>)
   &ScaleMeshSize(Scale Factor=&Multiply_VectorByFloat(
     Vector=<1 1 1>
     Float=&FloatFromCurve(FloatCurve=DataInterface() {propertyValues {Curve {keys [<at 0.78 1>,<at 1 0>]}}})
    ))
  ]
  rendererClasses [&NiagaraMeshRendererProperties]
  renderers [{rendererClass=&NiagaraMeshRendererProperties, rendererIndex=0}]
 }
}
summary {
 systemName=NS_Butter_Jack
 emitters [
  {emitterName=Empty, simTarget=GPUComputeSim, rendererClasses [&NiagaraMeshRendererProperties]}
  {emitterName=Empty001, rendererClasses [&NiagaraMeshRendererProperties]}
 ]
 userVariables [User.NewNiagaraDataInterfaceSkeletalMesh {
   type=NiagaraDataInterfaceSkeletalMesh
   defaultValue {
    struct=NiagaraExt_VariableValue_DataInterface
    value {dataInterface=&self:NiagaraDataInterfaceSkeletalMesh_0}
   }
  }]
}
