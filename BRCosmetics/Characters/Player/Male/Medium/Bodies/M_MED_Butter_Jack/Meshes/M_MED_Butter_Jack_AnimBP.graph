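# Post-copy AnimBP for the Butter Jack male body cosmetic, adding secondary motion to leg loops, strings and popcorn.
# It copies the attached parent's pose, then chains partial-alpha AnimDynamics nodes: angular springs for leg bones, damped linear motion for popcorn.
# Each update maps clavicle pitch relative to spine_05 into alphas for additive corrective rotations.
# Only the right clavicle correction reaches the output; the left node is unconnected.
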AnimBlueprint /BRCosmetics/Characters/Player/Male/Medium/Bodies/M_MED_Butter_Jack/Meshes/M_MED_Butter_Jack_AnimBP
target_skeleton=&/M_MED_Butter_Jack_Skeleton
graphs [
 AnimGraph {nodes [
   "Transform (Modify) Bone - Bone: dyn_clavicle_l_deltaRotate"(Alpha=Get L_clav_rotate_forward() -> (L_clav_rotate_forward:double), Translation="0,0,0", Rotation="0,22,0", Scale="1,1,1") {
    rotationMode=BMM_Additive
    rotationSpace=BCS_BoneSpace
    alphaBoolBlend {blendOption=Linear}
    alphaScaleBiasClamp {bInterpResult=true}
    boneToModify {boneName=dyn_clavicle_l_deltaRotate}
    rotation {yaw=22}
   }
   n1 = Anim Dynamics - Bone: dyn_L_leg_loop1_A(ComponentPose=Local To Component(LocalPose=Copy Pose From Mesh() {bUseAttachedParent=true}), Alpha=1, ExternalForce="0,0,0") {
    angularDampingOverride=0.9999
    angularSpringConstant=15
    bAngularSpring=true
    bGravityOverrideInSimSpace=true
    bOverrideAngularDamping=true
    bUseGravityOverride=true
    alphaBoolBlend {blendOption=Linear}
    alphaScaleBiasClamp {scale=0.75}
    boundBone {boneName=dyn_L_leg_loop1_A}
    componentLinearAccScale=<1 1 1>
    gravityOverride {z=-600}
    physicsBodyDefinitions [{
      sphereCollisionRadius=0
      boundBone {boneName=dyn_L_leg_loop1_A}
      boxExtents=<1 1 1>
      constraintSetup {angularLimitsMax=<0 25 10>, angularLimitsMin=<0 -25 -10>, angularTarget=<1 0 0>}
      localJointOffset=<0.01 0 0>
     }]
   }
   n2 = Anim Dynamics - Bone: dyn_L_leg_loop1_B(ComponentPose=@n1, Alpha=1, ExternalForce="0,0,0") {
    angularDampingOverride=0.9999
    angularSpringConstant=15
    bAngularSpring=true
    bGravityOverrideInSimSpace=true
    bOverrideAngularDamping=true
    bUseGravityOverride=true
    alphaBoolBlend {blendOption=Linear}
    alphaScaleBiasClamp {scale=0.75}
    boundBone {boneName=dyn_L_leg_loop1_B}
    componentLinearAccScale=<1 1 1>
    gravityOverride {z=-600}
    physicsBodyDefinitions [{
      sphereCollisionRadius=0
      boundBone {boneName=dyn_L_leg_loop1_B}
      boxExtents=<1 1 1>
      constraintSetup {angularLimitsMax=<0 25 10>, angularLimitsMin=<0 -25 -10>, angularTarget=<1 0 0>}
      localJointOffset=<0.01 0 0>
     }]
   }
   n3 = Anim Dynamics - Bone: dyn_L_leg_loop2_A(ComponentPose=@n2, Alpha=1, ExternalForce="0,0,0") {
    angularDampingOverride=0.9999
    angularSpringConstant=15
    bAngularSpring=true
    bGravityOverrideInSimSpace=true
    bOverrideAngularDamping=true
    bUseGravityOverride=true
    alphaBoolBlend {blendOption=Linear}
    alphaScaleBiasClamp {scale=0.75}
    boundBone {boneName=dyn_L_leg_loop2_A}
    componentLinearAccScale=<1 1 1>
    gravityOverride {z=-600}
    physicsBodyDefinitions [{
      sphereCollisionRadius=0
      boundBone {boneName=dyn_L_leg_loop2_A}
      boxExtents=<1 1 1>
      constraintSetup {angularLimitsMax=<0 25 10>, angularLimitsMin=<0 -25 -10>, angularTarget=<1 0 0>}
      localJointOffset=<0.01 0 0>
     }]
   }
   n4 = Anim Dynamics - Bone: dyn_L_leg_loop2_B(ComponentPose=@n3, Alpha=1, ExternalForce="0,0,0") {
    angularDampingOverride=0.9999
    angularSpringConstant=15
    bAngularSpring=true
    bGravityOverrideInSimSpace=true
    bOverrideAngularDamping=true
    bUseGravityOverride=true
    alphaBoolBlend {blendOption=Linear}
    alphaScaleBiasClamp {scale=0.75}
    boundBone {boneName=dyn_L_leg_loop2_B}
    componentLinearAccScale=<1 1 1>
    gravityOverride {z=-600}
    physicsBodyDefinitions [{
      sphereCollisionRadius=0
      boundBone {boneName=dyn_L_leg_loop2_B}
      boxExtents=<1 1 1>
      constraintSetup {angularLimitsMax=<0 25 10>, angularLimitsMin=<0 -25 -10>, angularTarget=<1 0 0>}
      localJointOffset=<0.01 0 0>
     }]
   }
   n5 = Anim Dynamics - Bone: dyn_L_leg_string1_A(ComponentPose=@n4, Alpha=1, ExternalForce="0,0,0") {
    angularDampingOverride=0.9999
    angularSpringConstant=15
    bAngularSpring=true
    bGravityOverrideInSimSpace=true
    bOverrideAngularDamping=true
    bUseGravityOverride=true
    alphaBoolBlend {blendOption=Linear}
    alphaScaleBiasClamp {scale=0.75}
    boundBone {boneName=dyn_L_leg_string1_A}
    componentLinearAccScale=<0.8 0.8 0.8>
    gravityOverride {z=-700}
    physicsBodyDefinitions [{
      sphereCollisionRadius=0
      boundBone {boneName=dyn_L_leg_string1_A}
      boxExtents=<1 1 1>
      constraintSetup {angularLimitsMax=<0 25 10>, angularLimitsMin=<0 -25 -10>, angularTarget=<1 0 0>}
      localJointOffset=<0.01 0 0>
     }]
   }
   n6 = Anim Dynamics - Bone: dyn_L_leg_string1_B(ComponentPose=@n5, Alpha=1, ExternalForce="0,0,0") {
    angularDampingOverride=0.9999
    angularSpringConstant=15
    bAngularSpring=true
    bGravityOverrideInSimSpace=true
    bOverrideAngularDamping=true
    bUseGravityOverride=true
    alphaBoolBlend {blendOption=Linear}
    alphaScaleBiasClamp {scale=0.75}
    boundBone {boneName=dyn_L_leg_string1_B}
    componentLinearAccScale=<0.8 0.8 0.8>
    gravityOverride {z=-700}
    physicsBodyDefinitions [{
      sphereCollisionRadius=0
      boundBone {boneName=dyn_L_leg_string1_B}
      boxExtents=<1 1 1>
      constraintSetup {angularLimitsMax=<0 25 10>, angularLimitsMin=<0 -25 -10>, angularTarget=<1 0 0>}
      localJointOffset=<0.01 0 0>
     }]
   }
   n7 = Anim Dynamics - Bone: dyn_L_leg_string2_A(ComponentPose=@n6, Alpha=1, ExternalForce="0,0,0") {
    angularDampingOverride=0.9999
    angularSpringConstant=15
    bAngularSpring=true
    bGravityOverrideInSimSpace=true
    bOverrideAngularDamping=true
    bUseGravityOverride=true
    alphaBoolBlend {blendOption=Linear}
    alphaScaleBiasClamp {scale=0.75}
    boundBone {boneName=dyn_L_leg_string2_A}
    componentLinearAccScale=<0.8 0.8 0.8>
    gravityOverride {z=-700}
    physicsBodyDefinitions [{
      sphereCollisionRadius=0
      boundBone {boneName=dyn_L_leg_string2_A}
      boxExtents=<1 1 1>
      constraintSetup {angularLimitsMax=<0 25 10>, angularLimitsMin=<0 -25 -10>, angularTarget=<1 0 0>}
      localJointOffset=<0.01 0 0>
     }]
   }
   n8 = Anim Dynamics - Bone: dyn_L_leg_string2_B(ComponentPose=@n7, Alpha=1, ExternalForce="0,0,0") {
    angularDampingOverride=0.9999
    angularSpringConstant=15
    bAngularSpring=true
    bGravityOverrideInSimSpace=true
    bOverrideAngularDamping=true
    bUseGravityOverride=true
    alphaBoolBlend {blendOption=Linear}
    alphaScaleBiasClamp {scale=0.75}
    boundBone {boneName=dyn_L_leg_string2_B}
    componentLinearAccScale=<0.8 0.8 0.8>
    gravityOverride {z=-700}
    physicsBodyDefinitions [{
      sphereCollisionRadius=0
      boundBone {boneName=dyn_L_leg_string2_B}
      boxExtents=<1 1 1>
      constraintSetup {angularLimitsMax=<0 25 10>, angularLimitsMin=<0 -25 -10>, angularTarget=<1 0 0>}
      localJointOffset=<0.01 0 0>
     }]
   }
   n9 = Anim Dynamics - Bone: dyn_R_leg_loop1_A(ComponentPose=@n8, Alpha=1, ExternalForce="0,0,0") {
    angularDampingOverride=0.9999
    angularSpringConstant=15
    bAngularSpring=true
    bGravityOverrideInSimSpace=true
    bOverrideAngularDamping=true
    bUseGravityOverride=true
    alphaBoolBlend {blendOption=Linear}
    alphaScaleBiasClamp {scale=0.75}
    boundBone {boneName=dyn_R_leg_loop1_A}
    componentLinearAccScale=<1 1 1>
    gravityOverride {z=-600}
    physicsBodyDefinitions [{
      sphereCollisionRadius=0
      boundBone {boneName=dyn_R_leg_loop1_A}
      boxExtents=<1 1 1>
      constraintSetup {angularLimitsMax=<0 25 10>, angularLimitsMin=<0 -25 -10>, angularTarget=<1 0 0>}
      localJointOffset=<-0.01 0 0>
     }]
   }
   n10 = Anim Dynamics - Bone: dyn_R_leg_loop1_B(ComponentPose=@n9, Alpha=1, ExternalForce="0,0,0") {
    angularDampingOverride=0.9999
    angularSpringConstant=15
    bAngularSpring=true
    bGravityOverrideInSimSpace=true
    bOverrideAngularDamping=true
    bUseGravityOverride=true
    alphaBoolBlend {blendOption=Linear}
    alphaScaleBiasClamp {scale=0.75}
    boundBone {boneName=dyn_R_leg_loop1_B}
    componentLinearAccScale=<1 1 1>
    gravityOverride {z=-600}
    physicsBodyDefinitions [{
      sphereCollisionRadius=0
      boundBone {boneName=dyn_R_leg_loop1_B}
      boxExtents=<1 1 1>
      constraintSetup {angularLimitsMax=<0 25 10>, angularLimitsMin=<0 -25 -10>, angularTarget=<1 0 0>}
      localJointOffset=<-0.01 0 0>
     }]
   }
   n11 = Anim Dynamics - Bone: dyn_R_leg_loop2_A(ComponentPose=@n10, Alpha=1, ExternalForce="0,0,0") {
    angularDampingOverride=0.9999
    angularSpringConstant=15
    bAngularSpring=true
    bGravityOverrideInSimSpace=true
    bOverrideAngularDamping=true
    bUseGravityOverride=true
    alphaBoolBlend {blendOption=Linear}
    alphaScaleBiasClamp {scale=0.75}
    boundBone {boneName=dyn_R_leg_loop2_A}
    componentLinearAccScale=<1 1 1>
    gravityOverride {z=-600}
    physicsBodyDefinitions [{
      sphereCollisionRadius=0
      boundBone {boneName=dyn_R_leg_loop2_A}
      boxExtents=<1 1 1>
      constraintSetup {angularLimitsMax=<0 25 10>, angularLimitsMin=<0 -25 -10>, angularTarget=<1 0 0>}
      localJointOffset=<-0.01 0 0>
     }]
   }
   n12 = Anim Dynamics - Bone: dyn_R_leg_loop2_B(ComponentPose=@n11, Alpha=1, ExternalForce="0,0,0") {
    angularDampingOverride=0.9999
    angularSpringConstant=15
    bAngularSpring=true
    bGravityOverrideInSimSpace=true
    bOverrideAngularDamping=true
    bUseGravityOverride=true
    alphaBoolBlend {blendOption=Linear}
    alphaScaleBiasClamp {scale=0.75}
    boundBone {boneName=dyn_R_leg_loop2_B}
    componentLinearAccScale=<1 1 1>
    gravityOverride {z=-600}
    physicsBodyDefinitions [{
      sphereCollisionRadius=0
      boundBone {boneName=dyn_R_leg_loop2_B}
      boxExtents=<1 1 1>
      constraintSetup {angularLimitsMax=<0 25 10>, angularLimitsMin=<0 -25 -10>, angularTarget=<1 0 0>}
      localJointOffset=<-0.01 0 0>
     }]
   }
   n13 = Anim Dynamics - Bone: dyn_R_leg_string1_A(ComponentPose=@n12, Alpha=1, ExternalForce="0,0,0") {
    angularDampingOverride=0.9999
    angularSpringConstant=15
    bAngularSpring=true
    bGravityOverrideInSimSpace=true
    bOverrideAngularDamping=true
    bUseGravityOverride=true
    alphaBoolBlend {blendOption=Linear}
    alphaScaleBiasClamp {scale=0.75}
    boundBone {boneName=dyn_R_leg_string1_A}
    componentLinearAccScale=<0.8 0.8 0.8>
    gravityOverride {z=-700}
    physicsBodyDefinitions [{
      sphereCollisionRadius=0
      boundBone {boneName=dyn_R_leg_string1_A}
      boxExtents=<1 1 1>
      constraintSetup {angularLimitsMax=<0 25 10>, angularLimitsMin=<0 -25 -10>, angularTarget=<1 0 0>}
      localJointOffset=<-0.01 0 0>
     }]
   }
   n14 = Anim Dynamics - Bone: dyn_R_leg_string1_B(ComponentPose=@n13, Alpha=1, ExternalForce="0,0,0") {
    angularDampingOverride=0.9999
    angularSpringConstant=15
    bAngularSpring=true
    bGravityOverrideInSimSpace=true
    bOverrideAngularDamping=true
    bUseGravityOverride=true
    alphaBoolBlend {blendOption=Linear}
    alphaScaleBiasClamp {scale=0.75}
    boundBone {boneName=dyn_R_leg_string1_B}
    componentLinearAccScale=<0.8 0.8 0.8>
    gravityOverride {z=-700}
    physicsBodyDefinitions [{
      sphereCollisionRadius=0
      boundBone {boneName=dyn_R_leg_string1_B}
      boxExtents=<1 1 1>
      constraintSetup {angularLimitsMax=<0 25 10>, angularLimitsMin=<0 -25 -10>, angularTarget=<1 0 0>}
      localJointOffset=<-0.01 0 0>
     }]
   }
   n15 = Anim Dynamics - Bone: dyn_R_leg_string2_A(ComponentPose=@n14, Alpha=1, ExternalForce="0,0,0") {
    angularDampingOverride=0.9999
    angularSpringConstant=15
    bAngularSpring=true
    bGravityOverrideInSimSpace=true
    bOverrideAngularDamping=true
    bUseGravityOverride=true
    alphaBoolBlend {blendOption=Linear}
    alphaScaleBiasClamp {scale=0.75}
    boundBone {boneName=dyn_R_leg_string2_A}
    componentLinearAccScale=<0.8 0.8 0.8>
    gravityOverride {z=-700}
    physicsBodyDefinitions [{
      sphereCollisionRadius=0
      boundBone {boneName=dyn_R_leg_string2_A}
      boxExtents=<1 1 1>
      constraintSetup {angularLimitsMax=<0 25 10>, angularLimitsMin=<0 -25 -10>, angularTarget=<1 0 0>}
      localJointOffset=<-0.01 0 0>
     }]
   }
   n16 = Anim Dynamics - Bone: dyn_R_leg_string2_B(ComponentPose=@n15, Alpha=1, ExternalForce="0,0,0") {
    angularDampingOverride=0.9999
    angularSpringConstant=15
    bAngularSpring=true
    bGravityOverrideInSimSpace=true
    bOverrideAngularDamping=true
    bUseGravityOverride=true
    alphaBoolBlend {blendOption=Linear}
    alphaScaleBiasClamp {scale=0.75}
    boundBone {boneName=dyn_R_leg_string2_B}
    componentLinearAccScale=<0.8 0.8 0.8>
    gravityOverride {z=-700}
    physicsBodyDefinitions [{
      sphereCollisionRadius=0
      boundBone {boneName=dyn_R_leg_string2_B}
      boxExtents=<1 1 1>
      constraintSetup {angularLimitsMax=<0 25 10>, angularLimitsMin=<0 -25 -10>, angularTarget=<1 0 0>}
      localJointOffset=<-0.01 0 0>
     }]
   }
   n17 = Anim Dynamics - Bone: dyn_popcorn_1(ComponentPose=@n16, Alpha=1, ExternalForce="0,0,0") {
    bGravityOverrideInSimSpace=true
    bOverrideLinearDamping=true
    bUseGravityOverride=true
    linearDampingOverride=0.99
    alphaBoolBlend {blendOption=Linear}
    alphaScaleBiasClamp {scale=0.25}
    boundBone {boneName=dyn_popcorn_1}
    componentLinearAccScale=<0.5 0.5 0.5>
    componentLinearVelScale=<0.01 0.01 0.01>
    gravityOverride {z=-600}
    physicsBodyDefinitions [{
      sphereCollisionRadius=0
      boundBone {boneName=dyn_popcorn_1}
      boxExtents=<1 1 1>
      constraintSetup {angularLimitsMax=<0 0 0>, angularLimitsMin=<0 0 0>, angularTarget=<0 0 0>, linearAxesMax=<0 0 25>}
      localJointOffset=<0 0 -0.1>
     }]
   }
   n18 = Anim Dynamics - Bone: dyn_popcorn_2(ComponentPose=@n17, Alpha=1, ExternalForce="0,0,0") {
    bGravityOverrideInSimSpace=true
    bOverrideLinearDamping=true
    bUseGravityOverride=true
    linearDampingOverride=0.99
    alphaBoolBlend {blendOption=Linear}
    alphaScaleBiasClamp {scale=0.25}
    boundBone {boneName=dyn_popcorn_2}
    componentLinearAccScale=<0.5 0.5 0.5>
    componentLinearVelScale=<0.01 0.01 0.01>
    gravityOverride {z=-600}
    physicsBodyDefinitions [{
      sphereCollisionRadius=0
      boundBone {boneName=dyn_popcorn_2}
      boxExtents=<1 1 1>
      constraintSetup {angularLimitsMax=<0 0 0>, angularLimitsMin=<0 0 0>, angularTarget=<0 0 0>, linearAxesMax=<0 0 40>}
      localJointOffset=<0 0 -0.1>
     }]
   }
   n19 = Anim Dynamics - Bone: dyn_popcorn_3(ComponentPose=@n18, Alpha=1, ExternalForce="0,0,0") {
    bGravityOverrideInSimSpace=true
    bOverrideLinearDamping=true
    bUseGravityOverride=true
    linearDampingOverride=0.99
    alphaBoolBlend {blendOption=Linear}
    alphaScaleBiasClamp {scale=0.25}
    boundBone {boneName=dyn_popcorn_3}
    componentLinearAccScale=<0.3 0.3 0.3>
    componentLinearVelScale=<0.01 0.01 0.01>
    gravityOverride {z=-600}
    physicsBodyDefinitions [{
      sphereCollisionRadius=0
      boundBone {boneName=dyn_popcorn_3}
      boxExtents=<1 1 1>
      constraintSetup {angularLimitsMax=<0 0 0>, angularLimitsMin=<0 0 0>, angularTarget=<0 0 0>, linearAxesMax=<0 0 30>}
      localJointOffset=<0 0 -0.1>
     }]
   }
   n20 = Anim Dynamics - Bone: dyn_popcorn_4(ComponentPose=@n19, Alpha=1, ExternalForce="0,0,0") {
    bGravityOverrideInSimSpace=true
    bOverrideLinearDamping=true
    bUseGravityOverride=true
    linearDampingOverride=0.99
    alphaBoolBlend {blendOption=Linear}
    alphaScaleBiasClamp {scale=0.19}
    boundBone {boneName=dyn_popcorn_4}
    componentLinearAccScale=<0.4 0.4 0.4>
    componentLinearVelScale=<0.01 0.01 0.01>
    gravityOverride {z=-600}
    physicsBodyDefinitions [{
      sphereCollisionRadius=0
      boundBone {boneName=dyn_popcorn_4}
      boxExtents=<1 1 1>
      constraintSetup {angularLimitsMax=<0 0 0>, angularLimitsMin=<0 0 0>, angularTarget=<0 0 0>, linearAxesMax=<0 0 30>}
      localJointOffset=<0 0 -0.1>
     }]
   }
   n21 = Anim Dynamics - Bone: dyn_popcorn_5(ComponentPose=@n20, Alpha=1, ExternalForce="0,0,0") {
    bGravityOverrideInSimSpace=true
    bOverrideLinearDamping=true
    bUseGravityOverride=true
    linearDampingOverride=0.99
    alphaBoolBlend {blendOption=Linear}
    alphaScaleBiasClamp {scale=0.15}
    boundBone {boneName=dyn_popcorn_5}
    componentLinearAccScale=<0.38 0.38 0.38>
    componentLinearVelScale=<0.01 0.01 0.01>
    gravityOverride {z=-600}
    physicsBodyDefinitions [{
      sphereCollisionRadius=0
      boundBone {boneName=dyn_popcorn_5}
      boxExtents=<1 1 1>
      constraintSetup {angularLimitsMax=<0 0 0>, angularLimitsMin=<0 0 0>, angularTarget=<0 0 0>, linearAxesMax=<0 0 30>}
      localJointOffset=<0 0 -0.1>
     }]
   }
   n22 = Anim Dynamics - Bone: dyn_popcorn_6(ComponentPose=@n21, Alpha=1, ExternalForce="0,0,0") {
    bGravityOverrideInSimSpace=true
    bOverrideLinearDamping=true
    bUseGravityOverride=true
    linearDampingOverride=0.99
    alphaBoolBlend {blendOption=Linear}
    alphaScaleBiasClamp {scale=0.19}
    boundBone {boneName=dyn_popcorn_6}
    componentLinearAccScale=<0.4 0.4 0.4>
    componentLinearVelScale=<0.01 0.01 0.01>
    gravityOverride {z=-600}
    physicsBodyDefinitions [{
      sphereCollisionRadius=0
      boundBone {boneName=dyn_popcorn_6}
      boxExtents=<1 1 1>
      constraintSetup {angularLimitsMax=<0 0 0>, angularLimitsMin=<0 0 0>, angularTarget=<0 0 0>, linearAxesMax=<0 0 40>}
      localJointOffset=<0 0 -0.1>
     }]
   }
   n23 = "Transform (Modify) Bone - Bone: dyn_clavicle_r_deltaRotate"(
    ComponentPose=@n22
    Alpha=Get R_clav_rotate_forward() -> (R_clav_rotate_forward:double)
    Translation="0,0,0"
    Rotation="0,22,0"
    Scale="1,1,1"
   ) {
    rotationMode=BMM_Additive
    rotationSpace=BCS_BoneSpace
    alphaBoolBlend {blendOption=Linear}
    alphaScaleBiasClamp {bInterpResult=true}
    boneToModify {boneName=dyn_clavicle_r_deltaRotate}
    rotation {yaw=22}
   }
   Output Pose(Result=Component To Local(ComponentPose=@n23)) {layerGroup=None}
  ]}
 EventGraph {nodes [
   Event BlueprintUpdateAnimation()
   > n1 = Sequence()
   n2 = GetOwnerSkeletalMeshComponent()
   n3 = BreakTransform(InTransform=GetDeltaTransformFromRefPose(self=@n2, BoneName=clavicle_r, BaseName=spine_05))
   > Set R_clav_rotate_forward(R_clav_rotate_forward:double=MapRangeClamped(Value=BreakRotator(InRot=@n3.Rotation).Pitch, InRangeA=15, InRangeB=28)) -> (Output_Get:double)
   n4 = BreakTransform(InTransform=GetDeltaTransformFromRefPose(self=@n2, BoneName=clavicle_l, BaseName=spine_05))
   Set L_clav_rotate_forward(execute=@n1.then_1, L_clav_rotate_forward:double=MapRangeClamped(Value=BreakRotator(InRot=@n4.Rotation).Pitch, InRangeA=20, InRangeB=50)) -> (Output_Get:double)
  ]}
]
variables [
 L_clav_rotate_forward {type=number}
 R_clav_rotate_forward {type=number}
]
